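# Sound cue for the ThawEpson AI's looping idle sound, classed as boss fight SFX.
# A Random node with a single entry feeds one looping wave player, with dedicated attenuation and concurrency assets for the loop.
# Bus sends feed sidechain volume buses for weapons, music, ambient and footsteps, weighted most toward ambient and least toward music.
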SoundCue /ThawEpsonAI/Assets/Sounds/IP_SC_ThawEpson_Idle_Loop
root=Random_1
nodes [
 {id=Random_1, children [WavePlayer_0], properties {bShouldExcludeFromBranchCulling=true, weights [1]}}
 {id=WavePlayer_0, properties {bLooping=true, soundWaveAssetPtr=/ThawEpsonAI/Assets/Sounds/Waves/IP_ThawEpson_Idle_Loop}}
]
properties {
 attenuationSettings=/ThawEpsonAI/Assets/Sounds/Settings/IP_ATT_ThawEpson_Loop
 maxDistance=5200
 priority=5
 soundClassObject=&Fort_BossFight_SFX
 volumeMultiplier=1
 busSends [
  {audioBus=&AB_Sidechain_FortWeapon_Vol, minSendLevel=0, sendLevel=0.6, soundSourceBus=null}
  {audioBus=&AB_Sidechain_FortMusic_Vol, minSendLevel=0, sendLevel=0.2, soundSourceBus=null}
  {audioBus=&AB_Sidechain_FortAmbient_Vol, minSendLevel=0, sendLevel=0.8, soundSourceBus=null}
  {audioBus=/Game/Sounds/Modulation/EnvelopeFollowers/Volume/AB_Sidechain_Footsteps_Vol, minSendLevel=0, sendLevel=0.6, soundSourceBus=null}
 ]
 concurrencyOverrides {bLimitToOwner=true, maxCount=1}
 concurrencySet [/ThawEpsonAI/Assets/Sounds/Settings/IP_CONC_ThawEpson_Loop]
}
