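# Sound cue for the ThawEpson AI's respawn sound, classed as boss-fight SFX.
# A Random node, excluded from branch culling, feeds a single respawn wave, with boosted volume and the character's hit-damage attenuation asset.
# A concurrency override limits it to one instance per owner, stopping the oldest, with a short retrigger time.
# Bus sends feed sidechain volume buses for weapons, music, ambience and footsteps.
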SoundCue /ThawEpsonAI/Assets/Sounds/IP_SC_ThawEpson_Respawn
root=Random_1
nodes [
 {id=Random_1, children [WavePlayer_0], properties {bShouldExcludeFromBranchCulling=true, weights [1]}}
 {id=WavePlayer_0, properties {soundWaveAssetPtr=/ThawEpsonAI/Assets/Sounds/Waves/IP_ThawEpson_Respawn_01}}
]
properties {
 attenuationSettings=/ThawEpsonAI/Assets/Sounds/Settings/IP_ATT_ThawEpson_HitDamage
 bOverrideConcurrency=true
 maxDistance=6000
 priority=5
 soundClassObject=&Fort_BossFight_SFX
 volumeMultiplier=1.5
 busSends [
  {audioBus=&AB_Sidechain_FortWeapon_Vol, minSendLevel=0, sendLevel=0.6, soundSourceBus=null}
  {audioBus=&AB_Sidechain_FortMusic_Vol, minSendLevel=0, sendLevel=0.2, soundSourceBus=null}
  {audioBus=&AB_Sidechain_FortAmbient_Vol, minSendLevel=0, sendLevel=0.8, soundSourceBus=null}
  {audioBus=/Game/Sounds/Modulation/EnvelopeFollowers/Volume/AB_Sidechain_Footsteps_Vol, minSendLevel=0, sendLevel=0.6, soundSourceBus=null}
 ]
 concurrencyOverrides {bLimitToOwner=true, maxCount=1, resolutionRule=StopOldest, retriggerTime=0.3}
}
