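# Niagara impact burst for ThawEpson damage hits, built from three short-lived emitters.
# Cube_Spread places blue-cyan mesh cubes through a grid scratch module, scaling them by a radial region mask against a lifetime curve to form a travelling wave.
# Hit_Flare draws a camera-offset sprite flare with two renderers and computes a fading blue glow.
# Sparks throws orange meshes that slow under drag and stretch with speed.
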NiagaraSystem /ThawEpsonAI/Assets/VFX/NS_ThawEpson_DamageImpact
dependencies {
 usedDataInterfaces [&NiagaraDataInterfaceCurve, &NiagaraDataInterfaceVectorCurve, &NiagaraDataInterfaceColorCurve]
 usedDynamicInputs [&MaskFloat, &ReturnNormalizedExecIndex]
 usedModules [&SystemState]
}
emitters {
 Cube_Spread {
  emitterSpawnScript [/SetVariables(Emitter.Radius=40) {isSetParameters=true}]
  emitterUpdateScript [
   &EmitterState(Life Cycle Mode=::NewEnumerator1, Loop Behavior=::NewEnumerator1, Loop Duration=1.5, UseLoopDelay=false)
   &SpawnBurst_Instantaneous(Spawn Count=50, Use Spawn Probability=false, UseLoopCountLimit=false)
  ]
  particleSpawnScript [
   &InitializeParticle(
    Lifetime=0.25
    Mesh Scale Mode=::NewEnumerator3
    Mesh Uniform Scale=10
    Sprite Size Mode=::NewEnumerator3
    Uniform Sprite Size=60
    UsePositionOffset=false
    Write Material Random=true
   )
   &ShapeLocation(
    MaskSpawnGroup=false
    Ring / Disc Distribution Mode:ENiagaraSphereDistributionMode=::NewEnumerator3
    Ring Radius=$Emitter.Radius
    Rotation Mode=::NewEnumerator1
    Shape Primitive=::NewEnumerator4
    UseOwnerScale=true
   )
   &ShapeLocation(
    Box / Plane Mode:ENiagara_BoxPlaneMode=::NewEnumerator1
    MaskSpawnGroup=false
    Override Seed=false
    Rotation Mode=::NewEnumerator1
    Shape Primitive=::NewEnumerator2
    UseOwnerScale=true
   ) {enabled=false, module=ShapeLocation001}
   &self:ScratchModule(Grid Range:NiagaraFloat=$Emitter.Radius, Grid Size=4) {module=ScratchModule}
  ]
  particleUpdateScript [
   &ParticleState()
   &RegionMask(Region Falloff=1, Region Radius:NiagaraFloat=$Emitter.Radius, UseOffset=false, UseRotation=false)
   &RegionMask(
    Region Dimensions:Vector3f=&Multiply_VectorByFloat(
     Vector=<1 100 100>
     Float=&FloatFromCurve(
      FloatCurve=DataInterface() {propertyValues {Curve {keys [{arriveTangent=2.9256, interpMode=RCIM_Cubic, leaveTangent=2.9256, tangentMode=RCTM_User, time=0, value=0}, <at 1 1>]}}}
      Scale Curve=10
     )
    )
    Region Type:ENiagaraDebugDrawShape=::NewEnumerator1
    UseOffset=false
    UseRotation=false
   ) {enabled=false, module=RegionMask001}
   &ScaleMeshSize(Scale Factor=&VectorFromFloat(Value=&Lerp_Float(
      A=1
      B=0
      Alpha=&Multiply_Float(
       A=1.5
       B=&Abs_Float(Float:NiagaraFloat=&Subtract_Float(
         A=$Output.RegionMask.RegionMask
         B=&FloatFromCurve(FloatCurve=DataInterface() {propertyValues {Curve {
             defaultValue=0
             keys [{arriveTangent=-2.065, interpMode=RCIM_Cubic, leaveTangent=-2.065, tangentMode=RCTM_User, time=0, value=1}, <at 0.8 0 RCIM_Cubic>]
            }}})
        ))
      )
     )))
   &ScaleMeshSize(
    Initial Mesh Scale:Vector3f=$Particles.Scale
    Scale Factor=&VectorFromFloat(Value=&FloatFromCurve(FloatCurve=DataInterface() {propertyValues {Curve {
         defaultValue=0
         keys [
          {interpMode=RCIM_Cubic, tangentMode=RCTM_User, time=0.5, value=1}
          {arriveTangent=-3.6378, interpMode=RCIM_Cubic, leaveTangent=-3.6378, tangentMode=RCTM_User, time=1, value=0}
         ]
        }}}))
   ) {module=ScaleMeshSize001}
   &ScaleSpriteSize(Uniform Curve Sprite Scale=DataInterface() {propertyValues {Curve {keys [
        {interpMode=RCIM_Cubic, tangentMode=RCTM_User, time=0, value=1}
        {arriveTangent=-2.1337, interpMode=RCIM_Cubic, leaveTangent=-2.1337, tangentMode=RCTM_User, time=1, value=0}
       ]}}}) {enabled=false}
   &DynamicMaterialParameters(
    Index 0 Param 1=50
    Index 0 Param 2=1
    Module.Index0Float0Write=true
    Module.Index0Float1Write=true
    Module.Index0Float2Write=false
    Module.Index0Float3Write=true
   )
   &InheritSourceMovement(
    Applied Position Scale=&VectorFromCurve(VectorCurve=DataInterface() {propertyValues {
       XCurve {keys [{interpMode=RCIM_Cubic, tangentMode=RCTM_User, time=0, value=1}, <at 1 0>]}
       YCurve {keys [{interpMode=RCIM_Cubic, tangentMode=RCTM_User, time=0, value=1}, <at 1 0>]}
       ZCurve {keys [{interpMode=RCIM_Cubic, tangentMode=RCTM_User, time=0, value=1}, <at 1 0>]}
      }})
    ApplyForce=false
    ApplyPosition=true
    ApplyPositionFalloff=false
   ) {enabled=false}
   &SolveForcesAndVelocity(Clamp Velocity=false, Limit Acceleration=false) {enabled=false}
   &ScaleColor(
    Linear Color Curve=DataInterface() {propertyValues {
      AlphaCurve {keys [<at 0 1>]}
      BlueCurve {keys [<at 0.069 2>,<at 0.6019 0.6484>]}
      GreenCurve {keys [<at 0.069 0.7945>,<at 0.6019 0.2511>]}
      RedCurve {keys [<at 0.069 0>,<at 0.6019 0>]}
     }}
    Scale Mode=2
   )
  ]
  rendererClasses [&NiagaraMeshRendererProperties]
  renderers [{rendererClass=&NiagaraMeshRendererProperties, rendererIndex=0}]
 }
 Hit_Flare {
  emitterUpdateScript [
   &EmitterState()
   &SpawnBurst_Instantaneous(Use Spawn Probability=false, UseLoopCountLimit=false)
  ]
  particleSpawnScript [&InitializeParticle(
    Color=<rgba 1 0.55 0.1 1>
    Lifetime=0.22
    Mesh Scale Mode=::NewEnumerator3
    Sprite Size Mode=::NewEnumerator3
    Uniform Sprite Size=50
    UsePositionOffset=true
    Write Material Random=true
   )]
  particleUpdateScript [
   &ParticleState()
   &ScaleSpriteSize(Uniform Curve Sprite Scale=DataInterface() {propertyValues {Curve {
       defaultValue=0
       keys [
        {arriveTangent=2.6925, interpMode=RCIM_Cubic, leaveTangent=2.6925, tangentMode=RCTM_User, time=0, value=0.65}
        {interpMode=RCIM_Cubic, tangentMode=RCTM_User, time=0.3, value=1}
        {tangentMode=RCTM_User, time=1, value=0}
       ]
      }}})
   &SolveForcesAndVelocity(Clamp Velocity=false, Limit Acceleration=false)
   &CameraOffset(Camera Offset Amount=50)
   &ScaleColor(
    Scale Alpha=&FloatFromCurve(
     FloatCurve=DataInterface() {propertyValues {Curve {
        defaultValue=0.9847
        keys [
         {arriveTangent=2.7084, interpMode=RCIM_Cubic, leaveTangent=2.7084, tangentMode=RCTM_User, time=0, value=0.6}
         {arriveTangent=-1.9732, interpMode=RCIM_Cubic, leaveTangent=-1.9732, tangentMode=RCTM_User, time=1, value=0}
        ]
       }}}
     Scale Curve=5
    )
    Scale RGB=&VectorFromFloat(Value=&FloatFromCurve(
      FloatCurve=DataInterface() {propertyValues {Curve {
         defaultValue=0.9847
         keys [
          {arriveTangent=2.7084, interpMode=RCIM_Cubic, leaveTangent=2.7084, tangentMode=RCTM_User, time=0, value=0.6}
          {arriveTangent=-1.9732, interpMode=RCIM_Cubic, leaveTangent=-1.9732, tangentMode=RCTM_User, time=1, value=0}
         ]
        }}}
      Scale Curve=1.5
     ))
    ScaleA=true
    ScaleRGB=true
   )
   &DynamicMaterialParameters(
    Index 1 Param 1=30
    Module.Param0WriteEnabled1=true
    Module.Param1WriteEnabled1=true
    Module.Param2WriteEnabled1=true
    Module.Param3WriteEnabled1=true
    Write Parameter Index 0=false
    Write Parameter Index 1=true
   )
   /SetVariables(Particles.GlowSize=&Multiply_Vector2D_ByFloat(Float=2, Vector2D=$Particles.SpriteSize)) {isSetParameters=true}
   /SetVariables(Particles.Glow Color=&MultiplyLinearColor(
     Linear Color=<rgba 0 0.7618 2 1>
     Scale Factor=&MakeVector4FromVectorAndFloat(
      "Vector (RGB)"=<1 1 1>
      "Float (Alpha)"=&FloatFromCurve(FloatCurve=DataInterface() {propertyValues {Curve {
          defaultValue=1
          keys [{interpMode=RCIM_Cubic, tangentMode=RCTM_User, time=0, value=1}, {interpMode=RCIM_Cubic, tangentMode=RCTM_User, time=1, value=0}]
         }}})
     )
    )) {isSetParameters=true}
  ]
  rendererClasses [&NiagaraSpriteRendererProperties]
  renderers [
   {rendererClass=&NiagaraSpriteRendererProperties}
   {rendererClass=&NiagaraSpriteRendererProperties, rendererIndex=1}
  ]
 }
 Sparks {
  emitterUpdateScript [
   &EmitterState(Life Cycle Mode=::NewEnumerator1, Loop Behavior=::NewEnumerator1, Loop Duration=1, UseLoopDelay=false)
   &SpawnBurst_Instantaneous(
    Spawn Count=&RandomRangeInt(Minimum=10, Maximum=15)
    Use Spawn Probability=false
    UseLoopCountLimit=false
   )
  ]
  particleSpawnScript [
   &PartitionParticles() {enabled=false}
   &FN_SetSpawnGroup(Number of Spawn Groups=2)
   &InitializeParticle(
    Color Maximum=<rgba 1 0.6667 0 1>
    Color Minimum=<rgba 1 0.2082 0 1>
    Color Mode=::NewEnumerator2
    Lifetime=&RandomRangeFloat(Minimum=0.25, Maximum=0.5)
    Mesh Scale Mode=::NewEnumerator1
    Mesh Uniform Scale Max=3
    Mesh Uniform Scale Min=1.5
    Sprite Size Mode=::NewEnumerator3
    Uniform Sprite Size=60
    UsePositionOffset=false
    Write Material Random=true
   )
   &ShapeLocation(
    MaskSpawnGroup=false
    Override Seed=false
    Ring Radius=&RandomRangeFloat(Minimum=10, Maximum=30)
    Rotation Mode=::NewEnumerator1
    Shape Primitive=::NewEnumerator4
    UseOwnerScale=true
   )
   &AddVelocity(
    Module.Use Falloff Distance=false
    Override Seed=false
    Velocity Mode=::NewEnumerator1
    Velocity Speed=&RandomRangeFloat(Minimum=222, Maximum=666)
   )
   &CameraOffset(Camera Offset Amount=10)
  ]
  particleUpdateScript [
   &ParticleState()
   &DynamicMaterialParameters(
    Index 0 Param 1=50
    Index 0 Param 2=1
    Module.Index0Float0Write=true
    Module.Index0Float1Write=true
    Module.Index0Float2Write=false
    Module.Index0Float3Write=true
   )
   &ScaleMeshSize(Scale Factor=&VectorFromFloat(Value=&FloatFromCurve(FloatCurve=DataInterface() {propertyValues {Curve {
         defaultValue=0.5
         keys [
          {arriveTangent=2.3191, interpMode=RCIM_Cubic, leaveTangent=2.3191, tangentMode=RCTM_User, time=0, value=0.6}
          {interpMode=RCIM_Cubic, tangentMode=RCTM_User, time=0.3, value=1}
          {tangentMode=RCTM_User, time=1, value=0}
         ]
        }}})))
   &Drag(Drag=15, Use Linear Drag=true, Use Rotational Drag=true)
   &GravityForce() {enabled=false}
   &SolveForcesAndVelocity(Clamp Velocity=false, Limit Acceleration=false)
   &ScaleMeshSizeBySpeed(Max Scale Factor=<30 1 1>, Speed Threshold=666)
   &ScaleColor(
    Scale RGB=&VectorFromFloat(Value=&FloatFromCurve(
      FloatCurve=DataInterface() {propertyValues {Curve {
         defaultValue=1
         keys [{interpMode=RCIM_Cubic, tangentMode=RCTM_User, time=0, value=1}, {interpMode=RCIM_Cubic, tangentMode=RCTM_User, time=1, value=0.5}]
        }}}
      Scale Curve=1.5
     ))
    ScaleA=true
    ScaleRGB=true
   )
  ]
  rendererClasses [&NiagaraMeshRendererProperties]
  renderers [{rendererClass=&NiagaraMeshRendererProperties, rendererIndex=0}]
 }
}
summary {
 systemName=NS_ThawEpson_DamageImpact
 emitters [
  {emitterName=Sparks, rendererClasses [&NiagaraMeshRendererProperties]}
  {emitterName=Cube_Spread, rendererClasses [&NiagaraMeshRendererProperties]}
  {emitterName=Hit_Flare, rendererClasses [&NiagaraSpriteRendererProperties]}
 ]
 userVariables [User.Color ID {type=NiagaraFloat, defaultValue=0, description=""}]
}
