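# Niagara system NS_ThawEpson_Idle, by its name an idle visual effect for ThawEpson, containing a single emitter named Glow.
# System spawn has a Set Variables assignment alongside the standard System State module.
# The system uses a color-array data interface and a Select Linear Color From Array dynamic input, and exposes an integer user parameter named Color Array.
# Glow lists no renderers.
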NiagaraSystem /ThawEpsonAI/Assets/VFX/NS_ThawEpson_Idle
dependencies {
 usedDataInterfaces [&NiagaraDataInterfaceArrayColor]
 usedDynamicInputs [&SelectLinearColorFromArray]
 usedModules [
  &SystemState
  &self:SystemSpawnScript.SystemScriptSource.SystemScriptGraph.NiagaraNodeAssignment_0.SetVariables_#1
 ]
}
emitters {Glow {rendererClasses [], renderers []}}
summary {systemName=NS_ThawEpson_Idle, emitters [{emitterName=Glow, rendererClasses []}], userVariables [User.Color Array {type=NiagaraInt32}]}
